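# Anchor setting entry for the control panel of the mobile HUD Layout Tool V2, for picking one of nine anchor positions.
# A setting-name label sits above a 3x3 uniform grid of layout tool buttons.
# The outer buttons share one arrow sprite, rotated by style icon angle to point outward, around a center "+" button.
# Widgets bind to the native HUDLayoutToolV2SettingsListEntry_Anchors parent; the Blueprint events are empty.
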WidgetBlueprint /FortniteCoreUI/Athena/HUD/Bacchus/HUDLayoutTool/ControlPanel/HUDLayoutToolV2_AnchorSetting
graphs [EventGraph {nodes [
   Event PreConstruct()
   Event Construct()
   Event Tick()
  ]}]
variables []
widgets {
 info {
  inheritedWidgetCount=11
  parentClass=/Script/FortniteUI.HUDLayoutToolV2SettingsListEntry_Anchors
  rootWidgetClass=&VerticalBox
  widgetCount=14
 }
 widgets [
  VerticalBox_313 {widgetClassPath=&VerticalBox, widget=VerticalBox_313}
  SizeBox_2 {
   widgetClassPath=&SizeBox
   parent=VerticalBox_313
   widget=SizeBox_2
   properties {widthOverride=600}
   slot {class=VerticalBoxSlot, properties {verticalAlignment=VAlign_Center, size {sizeRule=Fill, value=1}}}
  }
  Text_SettingName {
   widgetClassPath=&CommonTextBlock
   bInherited=true
   parent=SizeBox_2
   widget=Text_SettingName
   properties {
    scrollStyle=&TextScrollStyle-Base_C
    text="{Setting Name}"
    colorAndOpacity {colorUseRule=UseColor_Specified, specifiedColor=<rgba 0.4125 0.4125 0.4125 1>}
    font {
     bForceMonospaced=false
     bMaterialIsStencil=false
     fontMaterial=null
     fontObject=&BurbankSmall
     letterSpacing=0
     monospacedWidth=1
     size=22
     skewAmount=0
     typefaceFontName=Bold
     outlineSettings {
      bApplyOutlineToDropShadows=false
      bMiteredCorners=false
      bSeparateFillAlpha=false
      outlineMaterial=null
      outlineSize=0
      outlineColor=<rgba 0 0 0 1>
     }
    }
    shadowOffset=<0 0>
   }
   slot {class=SizeBoxSlot}
  }
  UniformGridPanel_72 {
   widgetClassPath=/Script/UMG.UniformGridPanel
   parent=VerticalBox_313
   widget=UniformGridPanel_72
   properties {slotPadding=<ltrb 5 5 5 5>}
   slot {class=VerticalBoxSlot, properties {horizontalAlignment=HAlign_Left, verticalAlignment=VAlign_Top}}
  }
  Button_TopLeft {
   widgetClassPath=/Game/Athena/HUD/Bacchus/HUDLayoutTool/HUDLayoutToolV2_Button
   bInherited=true
   bIsVariable=true
   parent=UniformGridPanel_72
   widget=Button_TopLeft
   properties {
    bOverride_HeightOverride=true
    bOverride_WidthOverride=true
    buttonHeightOverride=40
    buttonWidthOverride=40
    text=""
    textFontSize=0
    iconBrush {
     bIsDynamicallyLoaded=false
     drawAs=Image
     imageType=NoImage
     mirroring=NoMirror
     resourceName=None
     resourceObject=/Game/Athena/HUD/Bacchus/HUDLayoutTool/HLT_Icons/T-HLT-Arrow_Sprite
     tiling=NoTile
     imageSize=<25 25>
     margin=<ltrb 0 0 0 0>
     outlineSettings {
      bUseBrushTransparency=false
      roundingType=HalfHeightRadius
      width=0
      color {colorUseRule=UseColor_Specified, specifiedColor=<rgba 0 0 0 0>}
      cornerRadii=<0 0 0 0>
     }
     tintColor {colorUseRule=UseColor_Specified, specifiedColor=<rgba 1 1 1 1>}
     uVRegion {bIsValid=false, max=<0 0>, min=<0 0>}
    }
    internalPadding=<ltrb 0 0 0 0>
    normalButtonStyle {
     iconAngle=315
     backgroundColor=<rgba 0 0.117 1 1>
     iconColor=<rgba 1 1 1 1>
     textColor {colorUseRule=UseColor_Specified, specifiedColor=<rgba 1 1 1 1>}
    }
    pressedButtonStyle {
     iconAngle=315
     backgroundColor=<rgba 0.0331 0.2159 1 1>
     iconColor=<rgba 1 1 1 1>
     textColor {colorUseRule=UseColor_Specified, specifiedColor=<rgba 1 1 1 1>}
    }
   }
   slot {class=UniformGridSlot, properties {horizontalAlignment=HAlign_Center, verticalAlignment=VAlign_Center}}
  }
  Button_TopMiddle {
   widgetClassPath=&HUDLayoutToolV2_Button
   bInherited=true
   bIsVariable=true
   parent=UniformGridPanel_72
   widget=Button_TopMiddle
   properties {
    bOverride_HeightOverride=true
    bOverride_WidthOverride=true
    buttonHeightOverride=40
    buttonWidthOverride=40
    text=""
    textFontSize=0
    iconBrush {
     bIsDynamicallyLoaded=false
     drawAs=Image
     imageType=NoImage
     mirroring=NoMirror
     resourceName=None
     resourceObject=&T-HLT-Arrow_Sprite
     tiling=NoTile
     imageSize=<25 25>
     margin=<ltrb 0 0 0 0>
     outlineSettings {
      bUseBrushTransparency=false
      roundingType=HalfHeightRadius
      width=0
      color {colorUseRule=UseColor_Specified, specifiedColor=<rgba 0 0 0 0>}
      cornerRadii=<0 0 0 0>
     }
     tintColor {colorUseRule=UseColor_Specified, specifiedColor=<rgba 1 1 1 1>}
     uVRegion {bIsValid=false, max=<0 0>, min=<0 0>}
    }
    internalPadding=<ltrb 0 0 0 0>
   }
   slot {class=UniformGridSlot, properties {column=1, horizontalAlignment=HAlign_Center, verticalAlignment=VAlign_Center}}
  }
  Button_TopRight {
   widgetClassPath=&HUDLayoutToolV2_Button
   bInherited=true
   bIsVariable=true
   parent=UniformGridPanel_72
   widget=Button_TopRight
   properties {
    bOverride_HeightOverride=true
    bOverride_WidthOverride=true
    buttonHeightOverride=40
    buttonWidthOverride=40
    text=""
    textFontSize=0
    iconBrush {
     bIsDynamicallyLoaded=false
     drawAs=Image
     imageType=NoImage
     mirroring=NoMirror
     resourceName=None
     resourceObject=&T-HLT-Arrow_Sprite
     tiling=NoTile
     imageSize=<25 25>
     margin=<ltrb 0 0 0 0>
     outlineSettings {
      bUseBrushTransparency=false
      roundingType=HalfHeightRadius
      width=0
      color {colorUseRule=UseColor_Specified, specifiedColor=<rgba 0 0 0 0>}
      cornerRadii=<0 0 0 0>
     }
     tintColor {colorUseRule=UseColor_Specified, specifiedColor=<rgba 1 1 1 1>}
     uVRegion {bIsValid=false, max=<0 0>, min=<0 0>}
    }
    internalPadding=<ltrb 0 0 0 0>
    normalButtonStyle {
     iconAngle=45
     backgroundColor=<rgba 0 0.117 1 1>
     iconColor=<rgba 1 1 1 1>
     textColor {colorUseRule=UseColor_Specified, specifiedColor=<rgba 1 1 1 1>}
    }
    pressedButtonStyle {
     iconAngle=45
     backgroundColor=<rgba 0.0331 0.2159 1 1>
     iconColor=<rgba 1 1 1 1>
     textColor {colorUseRule=UseColor_Specified, specifiedColor=<rgba 1 1 1 1>}
    }
   }
   slot {class=UniformGridSlot, properties {column=2, horizontalAlignment=HAlign_Center, verticalAlignment=VAlign_Center}}
  }
  Button_MiddleLeft {
   widgetClassPath=&HUDLayoutToolV2_Button
   bInherited=true
   bIsVariable=true
   parent=UniformGridPanel_72
   widget=Button_MiddleLeft
   properties {
    bOverride_HeightOverride=true
    bOverride_WidthOverride=true
    buttonHeightOverride=40
    buttonWidthOverride=40
    text=""
    textFontSize=0
    iconBrush {
     bIsDynamicallyLoaded=false
     drawAs=Image
     imageType=NoImage
     mirroring=NoMirror
     resourceName=None
     resourceObject=&T-HLT-Arrow_Sprite
     tiling=NoTile
     imageSize=<25 25>
     margin=<ltrb 0 0 0 0>
     outlineSettings {
      bUseBrushTransparency=false
      roundingType=HalfHeightRadius
      width=0
      color {colorUseRule=UseColor_Specified, specifiedColor=<rgba 0 0 0 0>}
      cornerRadii=<0 0 0 0>
     }
     tintColor {colorUseRule=UseColor_Specified, specifiedColor=<rgba 1 1 1 1>}
     uVRegion {bIsValid=false, max=<0 0>, min=<0 0>}
    }
    internalPadding=<ltrb 0 0 0 0>
    normalButtonStyle {
     iconAngle=270
     backgroundColor=<rgba 0 0.117 1 1>
     iconColor=<rgba 1 1 1 1>
     textColor {colorUseRule=UseColor_Specified, specifiedColor=<rgba 1 1 1 1>}
    }
    pressedButtonStyle {
     iconAngle=270
     backgroundColor=<rgba 0.0331 0.2159 1 1>
     iconColor=<rgba 1 1 1 1>
     textColor {colorUseRule=UseColor_Specified, specifiedColor=<rgba 1 1 1 1>}
    }
   }
   slot {class=UniformGridSlot, properties {horizontalAlignment=HAlign_Center, row=1, verticalAlignment=VAlign_Center}}
  }
  Button_MiddleMiddle {
   widgetClassPath=&HUDLayoutToolV2_Button
   bInherited=true
   bIsVariable=true
   parent=UniformGridPanel_72
   widget=Button_MiddleMiddle
   properties {
    bOverride_HeightOverride=true
    bOverride_WidthOverride=true
    buttonHeightOverride=40
    buttonWidthOverride=40
    text=+
    iconBrush {
     bIsDynamicallyLoaded=false
     drawAs=NoDrawType
     imageType=NoImage
     mirroring=NoMirror
     resourceName=None
     resourceObject=null
     tiling=NoTile
     imageSize=<15 10>
     margin=<ltrb 0 0 0 0>
     outlineSettings {
      bUseBrushTransparency=false
      roundingType=HalfHeightRadius
      width=0
      color {colorUseRule=UseColor_Specified, specifiedColor=<rgba 0 0 0 0>}
      cornerRadii=<0 0 0 0>
     }
     tintColor {colorUseRule=UseColor_Specified, specifiedColor=<rgba 1 1 1 1>}
     uVRegion {bIsValid=false, max=<0 0>, min=<0 0>}
    }
    internalPadding=<ltrb 0 0 0 0>
   }
   slot {class=UniformGridSlot, properties {column=1, horizontalAlignment=HAlign_Center, row=1, verticalAlignment=VAlign_Center}}
  }
  Button_MiddleRight {
   widgetClassPath=&HUDLayoutToolV2_Button
   bInherited=true
   bIsVariable=true
   parent=UniformGridPanel_72
   widget=Button_MiddleRight
   properties {
    bOverride_HeightOverride=true
    bOverride_WidthOverride=true
    buttonHeightOverride=40
    buttonWidthOverride=40
    text=""
    textFontSize=0
    iconBrush {
     bIsDynamicallyLoaded=false
     drawAs=Image
     imageType=NoImage
     mirroring=NoMirror
     resourceName=None
     resourceObject=&T-HLT-Arrow_Sprite
     tiling=NoTile
     imageSize=<25 25>
     margin=<ltrb 0 0 0 0>
     outlineSettings {
      bUseBrushTransparency=false
      roundingType=HalfHeightRadius
      width=0
      color {colorUseRule=UseColor_Specified, specifiedColor=<rgba 0 0 0 0>}
      cornerRadii=<0 0 0 0>
     }
     tintColor {colorUseRule=UseColor_Specified, specifiedColor=<rgba 1 1 1 1>}
     uVRegion {bIsValid=false, max=<0 0>, min=<0 0>}
    }
    internalPadding=<ltrb 0 0 0 0>
    normalButtonStyle {
     iconAngle=90
     backgroundColor=<rgba 0 0.117 1 1>
     iconColor=<rgba 1 1 1 1>
     textColor {colorUseRule=UseColor_Specified, specifiedColor=<rgba 1 1 1 1>}
    }
    pressedButtonStyle {
     iconAngle=90
     backgroundColor=<rgba 0.0331 0.2159 1 1>
     iconColor=<rgba 1 1 1 1>
     textColor {colorUseRule=UseColor_Specified, specifiedColor=<rgba 1 1 1 1>}
    }
   }
   slot {class=UniformGridSlot, properties {column=2, horizontalAlignment=HAlign_Center, row=1, verticalAlignment=VAlign_Center}}
  }
  Button_BottomLeft {
   widgetClassPath=&HUDLayoutToolV2_Button
   bInherited=true
   bIsVariable=true
   parent=UniformGridPanel_72
   widget=Button_BottomLeft
   properties {
    bOverride_HeightOverride=true
    bOverride_WidthOverride=true
    buttonHeightOverride=40
    buttonWidthOverride=40
    text=""
    textFontSize=0
    iconBrush {
     bIsDynamicallyLoaded=false
     drawAs=Image
     imageType=NoImage
     mirroring=NoMirror
     resourceName=None
     resourceObject=&T-HLT-Arrow_Sprite
     tiling=NoTile
     imageSize=<25 25>
     margin=<ltrb 0 0 0 0>
     outlineSettings {
      bUseBrushTransparency=false
      roundingType=HalfHeightRadius
      width=0
      color {colorUseRule=UseColor_Specified, specifiedColor=<rgba 0 0 0 0>}
      cornerRadii=<0 0 0 0>
     }
     tintColor {colorUseRule=UseColor_Specified, specifiedColor=<rgba 1 1 1 1>}
     uVRegion {bIsValid=false, max=<0 0>, min=<0 0>}
    }
    internalPadding=<ltrb 0 0 0 0>
    normalButtonStyle {
     iconAngle=225
     backgroundColor=<rgba 0 0.117 1 1>
     iconColor=<rgba 1 1 1 1>
     textColor {colorUseRule=UseColor_Specified, specifiedColor=<rgba 1 1 1 1>}
    }
    pressedButtonStyle {
     iconAngle=225
     backgroundColor=<rgba 0.0331 0.2159 1 1>
     iconColor=<rgba 1 1 1 1>
     textColor {colorUseRule=UseColor_Specified, specifiedColor=<rgba 1 1 1 1>}
    }
   }
   slot {class=UniformGridSlot, properties {horizontalAlignment=HAlign_Center, row=2, verticalAlignment=VAlign_Center}}
  }
  Button_BottomMiddle {
   widgetClassPath=&HUDLayoutToolV2_Button
   bInherited=true
   bIsVariable=true
   parent=UniformGridPanel_72
   widget=Button_BottomMiddle
   properties {
    bOverride_HeightOverride=true
    bOverride_WidthOverride=true
    buttonHeightOverride=40
    buttonWidthOverride=40
    text=""
    textFontSize=0
    iconBrush {
     bIsDynamicallyLoaded=false
     drawAs=Image
     imageType=NoImage
     mirroring=NoMirror
     resourceName=None
     resourceObject=&T-HLT-Arrow_Sprite
     tiling=NoTile
     imageSize=<25 25>
     margin=<ltrb 0 0 0 0>
     outlineSettings {
      bUseBrushTransparency=false
      roundingType=HalfHeightRadius
      width=0
      color {colorUseRule=UseColor_Specified, specifiedColor=<rgba 0 0 0 0>}
      cornerRadii=<0 0 0 0>
     }
     tintColor {colorUseRule=UseColor_Specified, specifiedColor=<rgba 1 1 1 1>}
     uVRegion {bIsValid=false, max=<0 0>, min=<0 0>}
    }
    internalPadding=<ltrb 0 0 0 0>
    normalButtonStyle {
     iconAngle=180
     backgroundColor=<rgba 0 0.117 1 1>
     iconColor=<rgba 1 1 1 1>
     textColor {colorUseRule=UseColor_Specified, specifiedColor=<rgba 1 1 1 1>}
    }
    pressedButtonStyle {
     iconAngle=180
     backgroundColor=<rgba 0.0331 0.2159 1 1>
     iconColor=<rgba 1 1 1 1>
     textColor {colorUseRule=UseColor_Specified, specifiedColor=<rgba 1 1 1 1>}
    }
   }
   slot {class=UniformGridSlot, properties {column=1, horizontalAlignment=HAlign_Center, row=2, verticalAlignment=VAlign_Center}}
  }
  Button_BottomRight {
   widgetClassPath=&HUDLayoutToolV2_Button
   bInherited=true
   bIsVariable=true
   parent=UniformGridPanel_72
   widget=Button_BottomRight
   properties {
    bOverride_HeightOverride=true
    bOverride_WidthOverride=true
    buttonHeightOverride=40
    buttonWidthOverride=40
    text=""
    textFontSize=0
    iconBrush {
     bIsDynamicallyLoaded=false
     drawAs=Image
     imageType=NoImage
     mirroring=NoMirror
     resourceName=None
     resourceObject=&T-HLT-Arrow_Sprite
     tiling=NoTile
     imageSize=<25 25>
     margin=<ltrb 0 0 0 0>
     outlineSettings {
      bUseBrushTransparency=false
      roundingType=HalfHeightRadius
      width=0
      color {colorUseRule=UseColor_Specified, specifiedColor=<rgba 0 0 0 0>}
      cornerRadii=<0 0 0 0>
     }
     tintColor {colorUseRule=UseColor_Specified, specifiedColor=<rgba 1 1 1 1>}
     uVRegion {bIsValid=false, max=<0 0>, min=<0 0>}
    }
    internalPadding=<ltrb 0 0 0 0>
    normalButtonStyle {
     iconAngle=135
     backgroundColor=<rgba 0 0.117 1 1>
     iconColor=<rgba 1 1 1 1>
     textColor {colorUseRule=UseColor_Specified, specifiedColor=<rgba 1 1 1 1>}
    }
    pressedButtonStyle {
     iconAngle=135
     backgroundColor=<rgba 0.0331 0.2159 1 1>
     iconColor=<rgba 1 1 1 1>
     textColor {colorUseRule=UseColor_Specified, specifiedColor=<rgba 1 1 1 1>}
    }
   }
   slot {class=UniformGridSlot, properties {column=2, horizontalAlignment=HAlign_Center, row=2, verticalAlignment=VAlign_Center}}
  }
  Background {widgetClassPath=&UserWidget, bInherited=true, widget=null}
 ]
}
