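# Widget Blueprint for a boolean setting entry in the HUD Layout Tool V2 control panel, derived from HUDLayoutToolV2SettingsListEntry_Bool.
# It stacks the setting name text above a left-aligned value panel holding a toggle button with a checkmark icon, layered over a tinted tilted-square material border.
# The event graph is empty, so it only lays out and styles inherited widgets such as Text_SettingName and Button_ToggleBool for the native parent class.
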WidgetBlueprint /FortniteCoreUI/Athena/HUD/Bacchus/HUDLayoutTool/ControlPanel/HUDLayoutToolV2_BoolSetting
graphs [EventGraph {links [], nodes []}]
variables []
widgets {
 info {
  inheritedWidgetCount=4
  parentClass=/Script/FortniteUI.HUDLayoutToolV2SettingsListEntry_Bool
  rootWidgetClass=&VerticalBox
  widgetCount=8
 }
 widgets [
  VerticalBox_313 {widgetClassPath=&VerticalBox, widget=VerticalBox_313}
  SizeBox_2 {
   widgetClassPath=&SizeBox
   parent=VerticalBox_313
   widget=SizeBox_2
   properties {widthOverride=600}
   slot {class=VerticalBoxSlot, properties {verticalAlignment=VAlign_Center, size {sizeRule=Fill, value=1}}}
  }
  Text_SettingName {
   widgetClassPath=&CommonTextBlock
   bInherited=true
   parent=SizeBox_2
   widget=Text_SettingName
   properties {
    scrollStyle=&TextScrollStyle-Base_C
    text="{Setting Name}"
    colorAndOpacity {colorUseRule=UseColor_Specified, specifiedColor=<rgba 0.4125 0.4125 0.4125 1>}
    font {
     bForceMonospaced=false
     bMaterialIsStencil=false
     fontMaterial=null
     fontObject=&BurbankSmall
     letterSpacing=0
     monospacedWidth=1
     size=22
     skewAmount=0
     typefaceFontName=Bold
     outlineSettings {
      bApplyOutlineToDropShadows=false
      bMiteredCorners=false
      bSeparateFillAlpha=false
      outlineMaterial=null
      outlineSize=0
      outlineColor=<rgba 0 0 0 1>
     }
    }
    shadowOffset=<0 0>
   }
   slot {class=SizeBoxSlot}
  }
  Panel_Value {
   widgetClassPath=&HorizontalBox
   bInherited=true
   parent=VerticalBox_313
   widget=Panel_Value
   slot {class=VerticalBoxSlot, properties {horizontalAlignment=HAlign_Left, padding=<ltrb 20 6 0 0>}}
  }
  Overlay_6 {widgetClassPath=&Overlay, parent=Panel_Value, widget=Overlay_6, slot {class=HorizontalBoxSlot}}
  CheckboxBorder {
   widgetClassPath=&Image
   parent=Overlay_6
   widget=CheckboxBorder
   properties {
    brush {
     bIsDynamicallyLoaded=false
     drawAs=Image
     imageType=NoImage
     mirroring=NoMirror
     resourceName=None
     resourceObject=/Game/Athena/HUD/Bacchus/HUDLayoutTool/HLT_Icons/M_HUDLayoutToolV2_TiltSquare_Inst
     tiling=NoTile
     imageSize=<58 58>
     margin=<ltrb 0 0 0 0>
     outlineSettings {
      bUseBrushTransparency=false
      roundingType=HalfHeightRadius
      width=0
      color {colorUseRule=UseColor_Specified, specifiedColor=<rgba 0 0 0 0>}
      cornerRadii=<0 0 0 0>
     }
     tintColor {colorUseRule=UseColor_Specified, specifiedColor=<rgba 1 1 1 1>}
     uVRegion {bIsValid=false, max=<0 0>, min=<0 0>}
    }
    colorAndOpacity=<rgba 0 0.117 1 1>
   }
   slot {class=OverlaySlot, properties {horizontalAlignment=HAlign_Center, verticalAlignment=VAlign_Center}}
  }
  Button_ToggleBool {
   widgetClassPath=/Game/Athena/HUD/Bacchus/HUDLayoutTool/HUDLayoutToolV2_Button
   bInherited=true
   parent=Overlay_6
   widget=Button_ToggleBool
   properties {
    bOverride_HeightOverride=true
    bOverride_WidthOverride=true
    buttonHeightOverride=60
    buttonWidthOverride=60
    text=""
    backgroundBrush {
     bIsDynamicallyLoaded=false
     drawAs=NoDrawType
     imageType=NoImage
     mirroring=NoMirror
     resourceName=None
     resourceObject=null
     tiling=NoTile
     imageSize=<32 32>
     margin=<ltrb 0 0 0 0>
     outlineSettings {
      bUseBrushTransparency=false
      roundingType=HalfHeightRadius
      width=0
      color {colorUseRule=UseColor_Specified, specifiedColor=<rgba 0 0 0 0>}
      cornerRadii=<0 0 0 0>
     }
     tintColor {colorUseRule=UseColor_Specified, specifiedColor=<rgba 1 1 1 1>}
     uVRegion {bIsValid=false, max=<0 0>, min=<0 0>}
    }
    iconBrush {
     bIsDynamicallyLoaded=false
     drawAs=Image
     imageType=NoImage
     mirroring=NoMirror
     resourceName=None
     resourceObject=/Game/Athena/HUD/Bacchus/HUDLayoutTool/HLT_Icons/T-HLT-CheckMark-32_Sprite
     tiling=NoTile
     imageSize=<37 36>
     margin=<ltrb 0 0 0 0>
     outlineSettings {
      bUseBrushTransparency=false
      roundingType=HalfHeightRadius
      width=0
      color {colorUseRule=UseColor_Specified, specifiedColor=<rgba 0 0 0 0>}
      cornerRadii=<0 0 0 0>
     }
     tintColor {colorUseRule=UseColor_Specified, specifiedColor=<rgba 1 1 1 1>}
     uVRegion {bIsValid=false, max=<0 0>, min=<0 0>}
    }
    internalPadding=<ltrb 0 0 0 0>
   }
   slot {class=OverlaySlot, properties {horizontalAlignment=HAlign_Center, verticalAlignment=VAlign_Center}}
  }
  Background {widgetClassPath=&UserWidget, bInherited=true, widget=null}
 ]
}
