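# Settings rotator widget for the HUD Layout Tool V2 control panel, a FortSettingsRotator subclass that shows options as a thin material-drawn bar.
# A fixed-height SizeBox holds a Background border using a rotator material, and the inherited option text is collapsed.
# Option-populated, option-selected and default-option events push the count, selected index and default index into scalar parameters on the border's dynamic material, so the material renders the state.
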WidgetBlueprint /FortniteCoreUI/Athena/HUD/Bacchus/HUDLayoutTool/ControlPanel/HUDLayoutToolV2_SettingsRotator
graphs [EventGraph {nodes [
   n1 = Event On Options Populated!() -> (OutputDelegate, then, Count:int=0) {is=title K2Node_Event}
   > n2 = GetDynamicMaterial(self:Border Object Reference=Get Background() -> (Background:Border Object Reference))
   > SetScalarParameterValue(self=@n2, ParameterName=Options, Value="To Float (Integer)"(InInt=@n1.Count))
   n3 = Event On Options Selected!() -> (OutputDelegate, then, Index:int=0) {is=title K2Node_Event}
   > n4 = GetDynamicMaterial(self=Get Background())
   > SetScalarParameterValue(self=@n4, ParameterName=SelectedOption, Value="To Float (Integer)"(InInt=@n3.Index))
   n5 = Event On Default Option Specified!() -> (OutputDelegate, then, DefaultOptionIndex:int=0) {is=title K2Node_Event}
   > n6 = GetDynamicMaterial(self=Get Background())
   > SetScalarParameterValue(self=@n6, ParameterName=DefaultOption, Value="To Float (Integer)"(InInt=@n5.DefaultOptionIndex))
  ]}]
variables []
widgets {
 info {inheritedWidgetCount=2, parentClass=/Script/FortniteUI.FortSettingsRotator, rootWidgetClass=&SizeBox, widgetCount=4}
 widgets [
  SizeBox_1 {widgetClassPath=&SizeBox, widget=SizeBox_1, properties {bOverride_HeightOverride=true, heightOverride=15}}
  Background {
   widgetClassPath=&Border
   bIsVariable=true
   parent=SizeBox_1
   widget=Background
   properties {
    background {
     bIsDynamicallyLoaded=false
     drawAs=Image
     imageType=NoImage
     mirroring=NoMirror
     resourceName=None
     resourceObject=/Game/Athena/HUD/Bacchus/HUDLayoutTool/HLT_Icons/MI_HUDLayoutToolV2_SettingsRotator
     tiling=NoTile
     imageSize=<32 32>
     margin=<ltrb 0 0 0 0>
     outlineSettings {
      bUseBrushTransparency=false
      roundingType=HalfHeightRadius
      width=0
      color {colorUseRule=UseColor_Specified, specifiedColor=<rgba 0 0 0 0>}
      cornerRadii=<0 0 0 0>
     }
     tintColor {colorUseRule=UseColor_Specified, specifiedColor=<rgba 1 1 1 1>}
     uVRegion {bIsValid=false, max=<0 0>, min=<0 0>}
    }
    padding=<ltrb 0 0 0 0>
   }
   slot {class=SizeBoxSlot}
  }
  MyText {
   widgetClassPath=&CommonTextBlock
   bInherited=true
   parent=Background
   widget=MyText
   properties {
    justification=Center
    scrollStyle=&TextScrollStyle-Base_C
    text="{Option}"
    textTransformPolicy=ToUpper
    visibility=Collapsed
    colorAndOpacity {colorUseRule=UseColor_Specified, specifiedColor=<rgba 0.6105 0.8796 1 1>}
    font {
     bForceMonospaced=false
     bMaterialIsStencil=false
     fontMaterial=null
     fontObject=&BurbankSmall
     letterSpacing=0
     monospacedWidth=1
     size=16
     skewAmount=0
     typefaceFontName=Black
     outlineSettings {
      bApplyOutlineToDropShadows=false
      bMiteredCorners=false
      bSeparateFillAlpha=false
      outlineMaterial=null
      outlineSize=0
      outlineColor=<rgba 0 0 0 1>
     }
    }
    shadowOffset=<0 0>
   }
   slot {class=BorderSlot, properties {padding=<ltrb 0 0 0 0>}}
  }
  InputActionWidget {widgetClassPath=&CommonActionWidget, bInherited=true, widget=null}
 ]
}
