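# Widget Blueprint for a discrete setting entry in the HUD Layout Tool V2 control panel's settings list.
# A header row shows the setting name and uppercase value above a rotator flanked by decrease and increase buttons, which share one arrow sprite via opposite render rotations.
# The event graph is empty, so this asset supplies only layout and styling for widgets inherited from the C++ parent HUDLayoutToolV2SettingsListEntry_Discrete.
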WidgetBlueprint /FortniteCoreUI/Athena/HUD/Bacchus/HUDLayoutTool/ControlPanel/HUDLayoutToolV2_DiscreteSetting
graphs [EventGraph {links [], nodes []}]
variables []
widgets {
 info {
  inheritedWidgetCount=7
  parentClass=/Script/FortniteUI.HUDLayoutToolV2SettingsListEntry_Discrete
  rootWidgetClass=&VerticalBox
  widgetCount=12
 }
 widgets [
  VerticalBox_0 {widgetClassPath=&VerticalBox, widget=VerticalBox_0}
  HorizontalBox_0 {
   widgetClassPath=&HorizontalBox
   parent=VerticalBox_0
   widget=HorizontalBox_0
   slot {class=VerticalBoxSlot, properties {padding=<ltrb 0 0 90 0>}}
  }
  SizeBox_0 {
   widgetClassPath=&SizeBox
   parent=HorizontalBox_0
   widget=SizeBox_0
   properties {widthOverride=350}
   slot {class=HorizontalBoxSlot, properties {verticalAlignment=VAlign_Bottom, size {sizeRule=Fill, value=1}}}
  }
  Text_SettingName {
   widgetClassPath=&CommonTextBlock
   bInherited=true
   parent=SizeBox_0
   widget=Text_SettingName
   properties {
    scrollStyle=&TextScrollStyle-Base_C
    text="{Setting Name}"
    colorAndOpacity {colorUseRule=UseColor_Specified, specifiedColor=<rgba 0.4125 0.4125 0.4125 1>}
    font {
     bForceMonospaced=false
     bMaterialIsStencil=false
     fontMaterial=null
     fontObject=&BurbankSmall
     letterSpacing=0
     monospacedWidth=1
     size=22
     skewAmount=0
     typefaceFontName=Bold
     outlineSettings {
      bApplyOutlineToDropShadows=false
      bMiteredCorners=false
      bSeparateFillAlpha=false
      outlineMaterial=null
      outlineSize=0
      outlineColor=<rgba 0 0 0 1>
     }
    }
    shadowOffset=<0 0>
   }
   slot {class=SizeBoxSlot}
  }
  Text_PropertyValue {
   widgetClassPath=&CommonTextBlock
   bInherited=true
   parent=HorizontalBox_0
   widget=Text_PropertyValue
   properties {
    autoWrapText=true
    justification=Right
    scrollStyle=&TextScrollStyle-Base_C
    text="{Value}"
    textTransformPolicy=ToUpper
    font {
     bForceMonospaced=false
     bMaterialIsStencil=false
     fontMaterial=null
     fontObject=&BurbankSmall
     letterSpacing=0
     monospacedWidth=1
     size=24
     skewAmount=0
     typefaceFontName=Bold
     outlineSettings {
      bApplyOutlineToDropShadows=false
      bMiteredCorners=false
      bSeparateFillAlpha=false
      outlineMaterial=null
      outlineSize=0
      outlineColor=<rgba 0 0 0 1>
     }
    }
    renderTransform {angle=0, scale=<1 1>, shear=<0 0>, translation=<0 2>}
    shadowOffset=<0 0>
   }
   slot {class=HorizontalBoxSlot, properties {verticalAlignment=VAlign_Bottom, padding=<ltrb 40 0 0 0>}}
  }
  Panel_Value {
   widgetClassPath=&HorizontalBox
   bInherited=true
   parent=VerticalBox_0
   widget=Panel_Value
   slot {class=VerticalBoxSlot, properties {padding=<ltrb 0 6 0 0>}}
  }
  HorizontalBox {
   widgetClassPath=&HorizontalBox
   parent=Panel_Value
   widget=HorizontalBox
   slot {class=HorizontalBoxSlot, properties {size {sizeRule=Fill, value=1}}}
  }
  Button_DecreaseValue {
   widgetClassPath=/Game/Athena/HUD/Bacchus/HUDLayoutTool/HUDLayoutToolV2_Button
   bInherited=true
   parent=HorizontalBox
   widget=Button_DecreaseValue
   properties {
    bOverride_HeightOverride=true
    bOverride_WidthOverride=true
    buttonHeightOverride=80
    buttonWidthOverride=70
    text=""
    backgroundBrush {
     bIsDynamicallyLoaded=false
     drawAs=Border
     imageType=NoImage
     mirroring=NoMirror
     resourceName=None
     resourceObject=null
     tiling=NoTile
     imageSize=<32 32>
     margin=<ltrb 0 0 0 0>
     outlineSettings {
      bUseBrushTransparency=false
      roundingType=HalfHeightRadius
      width=0
      color {colorUseRule=UseColor_Specified, specifiedColor=<rgba 0 0 0 0>}
      cornerRadii=<0 0 0 0>
     }
     tintColor {colorUseRule=UseColor_Specified, specifiedColor=<rgba 1 1 1 1>}
     uVRegion {bIsValid=false, max=<0 0>, min=<0 0>}
    }
    iconBrush {
     bIsDynamicallyLoaded=false
     drawAs=Image
     imageType=NoImage
     mirroring=NoMirror
     resourceName=None
     resourceObject=/Game/Athena/HUD/Bacchus/HUDLayoutTool/HLT_Icons/T-HLT-Arrow_Sprite
     tiling=NoTile
     imageSize=<50 38>
     margin=<ltrb 0 0 0 0>
     outlineSettings {
      bUseBrushTransparency=false
      roundingType=HalfHeightRadius
      width=0
      color {colorUseRule=UseColor_Specified, specifiedColor=<rgba 0 0 0 0>}
      cornerRadii=<0 0 0 0>
     }
     tintColor {colorUseRule=UseColor_Specified, specifiedColor=<rgba 1 1 1 1>}
     uVRegion {bIsValid=false, max=<0 0>, min=<0 0>}
    }
    internalPadding=<ltrb 0 0 0 0>
    pressedButtonStyle {
     iconAngle=0
     backgroundColor=<rgba 1 1 1 1>
     iconColor=<rgba 0 0.117 1 1>
     textColor {colorUseRule=UseColor_Specified, specifiedColor=<rgba 1 1 1 1>}
    }
    renderTransform {angle=-90, scale=<1 1>, shear=<0 0>, translation=<5 0>}
   }
   slot {class=HorizontalBoxSlot, properties {size {sizeRule=Automatic, value=0.1}}}
  }
  Border_0 {
   widgetClassPath=&Border
   parent=HorizontalBox
   widget=Border_0
   properties {verticalAlignment=VAlign_Center, brushColor=<rgba 0 0 0 0.2>, padding=<ltrb 20 0 20 0>}
   slot {class=HorizontalBoxSlot, properties {padding=<ltrb 20 0 20 0>, size {sizeRule=Fill, value=1}}}
  }
  Rotator_SettingValue {
   widgetClassPath=&/HUDLayoutToolV2_SettingsRotator
   bInherited=true
   parent=Border_0
   widget=Rotator_SettingValue
   slot {class=BorderSlot, properties {verticalAlignment=VAlign_Center, padding=<ltrb 20 0 20 0>}}
  }
  Button_IncreaseValue {
   widgetClassPath=&HUDLayoutToolV2_Button
   bInherited=true
   parent=HorizontalBox
   widget=Button_IncreaseValue
   properties {
    bOverride_HeightOverride=true
    bOverride_WidthOverride=true
    buttonHeightOverride=80
    buttonWidthOverride=70
    text=""
    backgroundBrush {
     bIsDynamicallyLoaded=false
     drawAs=Border
     imageType=NoImage
     mirroring=NoMirror
     resourceName=None
     resourceObject=null
     tiling=NoTile
     imageSize=<32 32>
     margin=<ltrb 0 0 0 0>
     outlineSettings {
      bUseBrushTransparency=false
      roundingType=HalfHeightRadius
      width=0
      color {colorUseRule=UseColor_Specified, specifiedColor=<rgba 0 0 0 0>}
      cornerRadii=<0 0 0 0>
     }
     tintColor {colorUseRule=UseColor_Specified, specifiedColor=<rgba 1 1 1 1>}
     uVRegion {bIsValid=false, max=<0 0>, min=<0 0>}
    }
    iconBrush {
     bIsDynamicallyLoaded=false
     drawAs=Image
     imageType=NoImage
     mirroring=NoMirror
     resourceName=None
     resourceObject=&T-HLT-Arrow_Sprite
     tiling=NoTile
     imageSize=<50 38>
     margin=<ltrb 0 0 0 0>
     outlineSettings {
      bUseBrushTransparency=false
      roundingType=HalfHeightRadius
      width=0
      color {colorUseRule=UseColor_Specified, specifiedColor=<rgba 0 0 0 0>}
      cornerRadii=<0 0 0 0>
     }
     tintColor {colorUseRule=UseColor_Specified, specifiedColor=<rgba 1 1 1 1>}
     uVRegion {bIsValid=false, max=<0 0>, min=<0 0>}
    }
    internalPadding=<ltrb 0 0 0 0>
    pressedButtonStyle {
     iconAngle=0
     backgroundColor=<rgba 1 1 1 1>
     iconColor=<rgba 0 0.117 1 1>
     textColor {colorUseRule=UseColor_Specified, specifiedColor=<rgba 1 1 1 1>}
    }
    renderTransform {angle=90, scale=<1 1>, shear=<0 0>, translation=<-5 0>}
   }
   slot {class=HorizontalBoxSlot, properties {size {sizeRule=Automatic, value=0.1}}}
  }
  Background {widgetClassPath=&UserWidget, bInherited=true, widget=null}
 ]
}
