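# Widget Blueprint for a HUD Layout Tool V2 settings list entry that shows quickbar slot lock controls.
# A vertical stack places a grey setting name label above a dynamic entry box of QuickbarSlotLock widgets, spaced horizontally.
# It has no graph logic or variables; the label and entry box are inherited bindings from the native QuickbarLocks entry class, so this asset only supplies layout and styling.
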WidgetBlueprint /FortniteCoreUI/Athena/HUD/Bacchus/HUDLayoutTool/ControlPanel/HUDLayoutToolV2_QuickbarSlotLocks
graphs [EventGraph {links [], nodes []}]
variables []
widgets {
 info {
  inheritedWidgetCount=3
  parentClass=/Script/FortniteUI.HUDLayoutToolV2SettingsListEntry_QuickbarLocks
  rootWidgetClass=&VerticalBox
  widgetCount=5
 }
 widgets [
  VerticalBox_313 {widgetClassPath=&VerticalBox, widget=VerticalBox_313}
  SizeBox_2 {
   widgetClassPath=&SizeBox
   parent=VerticalBox_313
   widget=SizeBox_2
   properties {widthOverride=600}
   slot {class=VerticalBoxSlot, properties {verticalAlignment=VAlign_Center, size {sizeRule=Fill, value=1}}}
  }
  Text_SettingName {
   widgetClassPath=&CommonTextBlock
   bInherited=true
   parent=SizeBox_2
   widget=Text_SettingName
   properties {
    scrollStyle=&TextScrollStyle-Base_C
    text="{Setting Name}"
    colorAndOpacity {colorUseRule=UseColor_Specified, specifiedColor=<rgba 0.4125 0.4125 0.4125 1>}
    font {
     bForceMonospaced=false
     bMaterialIsStencil=false
     fontMaterial=null
     fontObject=&BurbankSmall
     letterSpacing=0
     monospacedWidth=1
     size=22
     skewAmount=0
     typefaceFontName=Bold
     outlineSettings {
      bApplyOutlineToDropShadows=false
      bMiteredCorners=false
      bSeparateFillAlpha=false
      outlineMaterial=null
      outlineSize=0
      outlineColor=<rgba 0 0 0 1>
     }
    }
    shadowOffset=<0 0>
   }
   slot {class=SizeBoxSlot}
  }
  EntryBox_SlotLocks {
   widgetClassPath=/Script/FortniteUI.FortDynamicEntryBox
   bInherited=true
   parent=VerticalBox_313
   widget=EntryBox_SlotLocks
   properties {
    entryWidgetClass=&/HUDLayoutToolV2_QuickbarSlotLock.HUDLayoutToolV2_QuickbarSlotLock_C
    numDesignerPreviewEntries=5
    entrySpacing=<30 0>
   }
   slot {class=VerticalBoxSlot, properties {padding=<ltrb 0 6 0 0>}}
  }
  Background {widgetClassPath=&UserWidget, bInherited=true, widget=null}
 ]
}
