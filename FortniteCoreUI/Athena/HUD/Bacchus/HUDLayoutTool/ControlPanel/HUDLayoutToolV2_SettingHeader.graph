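# Widget Blueprint for a section header row in the settings list of the HUD Layout Tool V2 control panel.
# A SizeBox sets a minimum height and bottom-aligns the inherited Text_SettingName, styled as uppercase blue Burbank with a render-transform shear, scroll style and larger mobile font.
# It has no graph logic of its own, relying on FortSettingsListEntry_Setting, and leaves the inherited Background slot empty.
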WidgetBlueprint /FortniteCoreUI/Athena/HUD/Bacchus/HUDLayoutTool/ControlPanel/HUDLayoutToolV2_SettingHeader
graphs [EventGraph {links [], nodes []}]
variables []
widgets {
 info {inheritedWidgetCount=2, parentClass=/Script/FortniteUI.FortSettingsListEntry_Setting, rootWidgetClass=&SizeBox, widgetCount=3}
 widgets [
  SizeBox_0 {widgetClassPath=&SizeBox, widget=SizeBox_0, properties {bOverride_MinDesiredHeight=true, minDesiredHeight=65}}
  Text_SettingName {
   widgetClassPath=&CommonTextBlock
   bInherited=true
   bIsVariable=true
   parent=SizeBox_0
   widget=Text_SettingName
   properties {
    mobileFontSizeMultiplier=1.5
    scrollStyle=&TextScrollStyle-Base_C
    text="{Header Text}"
    textTransformPolicy=ToUpper
    colorAndOpacity {colorUseRule=UseColor_Specified, specifiedColor=<rgba 0.0307 0.3813 0.9823 1>}
    font {
     bForceMonospaced=false
     bMaterialIsStencil=false
     fontMaterial=null
     fontObject=&Burbank
     letterSpacing=0
     monospacedWidth=1
     size=24
     skewAmount=0
     typefaceFontName=BurbankRegular-Black
     outlineSettings {
      bApplyOutlineToDropShadows=false
      bMiteredCorners=false
      bSeparateFillAlpha=false
      outlineMaterial=null
      outlineSize=0
      outlineColor=<rgba 0 0 0 1>
     }
    }
    renderTransform {angle=0, scale=<1 1>, shear=<-8 0>, translation=<0 0>}
    shadowOffset=<0 0>
   }
   slot {class=SizeBoxSlot, properties {verticalAlignment=VAlign_Bottom}}
  }
  Background {widgetClassPath=&UserWidget, bInherited=true, widget=null}
 ]
}
