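# Widget Blueprint for a control panel button in the mobile HUD Layout Tool V2, representing one layout context.
# It shows a layout name above a toggleable edit button, with a vehicle-mech placeholder icon, and a hidden-icon visibility button whose states are controlled externally.
# A collapsed highlight image overlays the button row for selection.
# The event graph is empty; the native HUDLayoutToolV2_ContextButton parent declares the key widgets and supplies any behavior.
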WidgetBlueprint /FortniteCoreUI/Athena/HUD/Bacchus/HUDLayoutTool/ControlPanel/HUDLayoutToolV2_ContextButton
graphs [EventGraph {links [], nodes []}]
variables []
widgets {
 info {inheritedWidgetCount=4, parentClass=/Script/FortniteUI.HUDLayoutToolV2_ContextButton, rootWidgetClass=&VerticalBox, widgetCount=8}
 widgets [
  VerticalBox_0 {widgetClassPath=&VerticalBox, widget=VerticalBox_0}
  SizeBox_0 {
   widgetClassPath=&SizeBox
   parent=VerticalBox_0
   widget=SizeBox_0
   properties {bOverride_WidthOverride=true, widthOverride=260}
   slot {class=VerticalBoxSlot, properties {verticalAlignment=VAlign_Center, size {sizeRule=Fill, value=1}}}
  }
  Text_LayoutName {
   widgetClassPath=&CommonTextBlock
   bInherited=true
   parent=SizeBox_0
   widget=Text_LayoutName
   properties {
    scrollStyle=&TextScrollStyle-Base_C
    text="{Layout Name}"
    colorAndOpacity {colorUseRule=UseColor_Specified, specifiedColor=<rgba 0.4125 0.4125 0.4125 1>}
    font {
     bForceMonospaced=false
     bMaterialIsStencil=false
     fontMaterial=null
     fontObject=&BurbankSmall
     letterSpacing=0
     monospacedWidth=1
     size=22
     skewAmount=0
     typefaceFontName=Bold
     outlineSettings {
      bApplyOutlineToDropShadows=false
      bMiteredCorners=false
      bSeparateFillAlpha=false
      outlineMaterial=null
      outlineSize=0
      outlineColor=<rgba 0 0 0 1>
     }
    }
    shadowOffset=<0 0>
   }
   slot {class=SizeBoxSlot}
  }
  Overlay_3 {
   widgetClassPath=&Overlay
   parent=VerticalBox_0
   widget=Overlay_3
   slot {class=VerticalBoxSlot, properties {padding=<ltrb 0 6 0 0>}}
  }
  HorizontalBox_3 {widgetClassPath=&HorizontalBox, parent=Overlay_3, widget=HorizontalBox_3, slot {class=OverlaySlot}}
  Button_ToggleContextEditing {
   widgetClassPath=/Game/Athena/HUD/Bacchus/HUDLayoutTool/HUDLayoutToolV2_Button
   bInherited=true
   parent=HorizontalBox_3
   widget=Button_ToggleContextEditing
   properties {
    bIsToggleable=true
    bMatchIconSize=true
    bOverride_HeightOverride=true
    bOverride_WidthOverride=true
    buttonHeightOverride=80
    buttonWidthOverride=156
    text=""
    visibility=Visible
    iconBrush {
     bIsDynamicallyLoaded=false
     drawAs=Image
     imageType=FullColor
     mirroring=NoMirror
     resourceName=None
     resourceObject=/Game/Athena/HUD/Bacchus/HUDLayoutTool/HLT_Icons/T-HLT-Vehicle-Mech
     tiling=NoTile
     imageSize=<45 60>
     margin=<ltrb 0 0 0 0>
     outlineSettings {
      bUseBrushTransparency=false
      roundingType=HalfHeightRadius
      width=0
      color {colorUseRule=UseColor_Specified, specifiedColor=<rgba 0 0 0 0>}
      cornerRadii=<0 0 0 0>
     }
     tintColor {colorUseRule=UseColor_Specified, specifiedColor=<rgba 1 1 1 1>}
     uVRegion {bIsValid=false, max=<0 0>, min=<0 0>}
    }
    internalPadding=<ltrb 0 0 0 0>
    selectedButtonStyle {
     iconAngle=0
     backgroundColor=<rgba 0.9301 1 0.0097 1>
     iconColor=<rgba 0 0 0 1>
     textColor {colorUseRule=UseColor_Specified, specifiedColor=<rgba 0 0 0 1>}
    }
   }
   slot {class=HorizontalBoxSlot, properties {horizontalAlignment=HAlign_Left, verticalAlignment=VAlign_Top, padding=<ltrb 0 0 10 0>}}
  }
  Button_ContextVisibility {
   widgetClassPath=&HUDLayoutToolV2_Button
   bInherited=true
   parent=HorizontalBox_3
   widget=Button_ContextVisibility
   properties {
    bExternallyControlledStates=true
    bMatchIconSize=true
    bOverride_HeightOverride=true
    bOverride_WidthOverride=true
    buttonHeightOverride=80
    buttonWidthOverride=94
    text=""
    visibility=Visible
    iconBrush {
     bIsDynamicallyLoaded=false
     drawAs=Image
     imageType=FullColor
     mirroring=NoMirror
     resourceName=None
     resourceObject=/Game/Athena/HUD/Bacchus/HUDLayoutTool/HLT_Icons/T-HLT-Hidden
     tiling=NoTile
     imageSize=<55 43>
     margin=<ltrb 0 0 0 0>
     outlineSettings {
      bUseBrushTransparency=false
      roundingType=HalfHeightRadius
      width=0
      color {colorUseRule=UseColor_Specified, specifiedColor=<rgba 0 0 0 0>}
      cornerRadii=<0 0 0 0>
     }
     tintColor {colorUseRule=UseColor_Specified, specifiedColor=<rgba 1 1 1 1>}
     uVRegion {bIsValid=false, max=<0 0>, min=<0 0>}
    }
    internalPadding=<ltrb 0 0 0 0>
    normalButtonStyle {
     iconAngle=0
     backgroundColor=<rgba 0 0 0 0.2>
     iconColor=<rgba 1 1 1 0.3>
     textColor {colorUseRule=UseColor_Specified, specifiedColor=<rgba 1 1 1 1>}
    }
    pressedButtonStyle {
     iconAngle=0
     backgroundColor=<rgba 1 1 1 0.2>
     iconColor=<rgba 1 1 1 1>
     textColor {colorUseRule=UseColor_Specified, specifiedColor=<rgba 1 1 1 1>}
    }
    selectedButtonStyle {
     iconAngle=0
     backgroundColor=<rgba 0 0 0 0.2>
     iconColor=<rgba 1 1 1 0.5>
     textColor {colorUseRule=UseColor_Specified, specifiedColor=<rgba 0 0 0 1>}
    }
   }
   slot {class=HorizontalBoxSlot, properties {horizontalAlignment=HAlign_Center, verticalAlignment=VAlign_Center}}
  }
  Image_SelectionHighlight {
   widgetClassPath=&Image
   bInherited=true
   parent=Overlay_3
   widget=Image_SelectionHighlight
   properties {
    visibility=Collapsed
    brush {
     bIsDynamicallyLoaded=false
     drawAs=NoDrawType
     imageType=NoImage
     mirroring=NoMirror
     resourceName=None
     resourceObject=/Game/Athena/HUD/Bacchus/HUDLayoutTool/HLT_Icons/M_UI_Square_Inst
     tiling=NoTile
     imageSize=<156 32>
     margin=<ltrb 0.25 0.25 0.25 0.25>
     outlineSettings {
      bUseBrushTransparency=false
      roundingType=HalfHeightRadius
      width=0
      color {colorUseRule=UseColor_Specified, specifiedColor=<rgba 0 0 0 0>}
      cornerRadii=<0 0 0 0>
     }
     tintColor {colorUseRule=UseColor_Specified, specifiedColor=<rgba 1 1 1 1>}
     uVRegion {bIsValid=false, max=<0 0>, min=<0 0>}
    }
    colorAndOpacity=<rgba 0.9301 1 0.0097 1>
   }
   slot {class=OverlaySlot, properties {verticalAlignment=VAlign_Fill}}
  }
 ]
}
